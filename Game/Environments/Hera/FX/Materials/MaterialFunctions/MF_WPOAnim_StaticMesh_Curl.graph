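# VFX material function that outputs a curl-style World Position Offset for static meshes, meant to be added to other WPO functions.
# Actor-relative world position, scrolled by looping time times speed and divided by tiling, feeds Vector Noise set to Perlin 3D rather than true Curl.
# Red vertex color, remapped between min and max inputs as ClampVertexColor, drives per-vertex variation through named reroutes, blending low/high intensity and tiling and scaling the offset.
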MaterialFunction /Game/Environments/Hera/FX/Materials/MaterialFunctions/MF_WPOAnim_StaticMesh_Curl
graph {
 properties {
  description="VFX Curl noise, ADD to other WPO functions, connect to WPO"
  userExposedCaption=VFX Static Mesh WPO Curl Noise
  libraryCategoriesText [VFX]
 }
 nodes [
  n1 = FunctionInput() {
   bUsePreviewValueAsDefault=true
   description=Amount of Curl
   inputName=Curl_Intensity
   inputType=FunctionInput_Scalar
   sortPriority=2
   previewValue=<50 0 0 1>
  }
  n2 = FunctionInput() {
   bUsePreviewValueAsDefault=true
   description="Amount of Curl, upper value"
   inputName=Curl_Intensity_High
   inputType=FunctionInput_Scalar
   sortPriority=3
   previewValue=<100 0 0 1>
  }
  n3 = LinearInterpolate(A=@n1, B=@n2, Alpha=NamedRerouteUsage())
  n4 = NamedRerouteUsage()
  n5 = WorldPosition()
  n6 = ActorPositionWS()
  n7 = FunctionInput() {
   bUsePreviewValueAsDefault=true
   description=Input looping time here to make movement looping
   inputName=InputLoopingTimeHere
   inputType=FunctionInput_Scalar
   previewValue=<1 0 0 1>
  }
  n8 = FunctionInput() {
   bUsePreviewValueAsDefault=true
   description=Speed of Curl Noise Worldspace movement
   inputName=CurlSpeed
   inputType=FunctionInput_Scalar
   sortPriority=1
   previewValue=<1 0 0 1>
  }
  n9 = FunctionInput() {
   bUsePreviewValueAsDefault=true
   description="Tiling of Curl Noise "
   inputName=CurlTiling
   inputType=FunctionInput_Scalar
   sortPriority=5
   previewValue=<250 0 0 1>
  }
  n10 = FunctionInput() {
   bUsePreviewValueAsDefault=true
   description="Tiling of Curl Noise, upper value "
   inputName=CurlTiling_High
   inputType=FunctionInput_Scalar
   sortPriority=6
   previewValue=<500 0 0 1>
  }
  n11 = Divide(A=Subtract(A=Subtract(A=@n5.XYZ, B=@n6), B=Multiply(A=@n7, B=@n8)), B=LinearInterpolate(A=@n9, B=@n10, Alpha=NamedRerouteUsage()))
  FunctionOutput(None=Multiply(A=Multiply(A=@n3, B=@n4), B=VectorNoise(World Position=@n11) {noiseFunction=VNF_VectorALU})) {description=Add together and connect to WPO, outputName=WPO Curl}
  n12 = FunctionInput() {
   bUsePreviewValueAsDefault=true
   description="Min value to set vertex color to, controls all other random values"
   inputName=VertexColor_LerpMin
   inputType=FunctionInput_Scalar
   sortPriority=1
   previewValue=<0 0 0 0>
  }
  n13 = FunctionInput() {
   bUsePreviewValueAsDefault=true
   description="Max value to set vertex color to, controls all other random values"
   inputName=VertexColor_LerpMax
   inputType=FunctionInput_Scalar
   previewValue=<1 0 0 0>
  }
  NamedRerouteDeclaration(None=LinearInterpolate(A=@n12, B=@n13, Alpha=VertexColor().R)) {name=ClampVertexColor, nodeColor=<rgba 0 0.1797 1 1>}
 ]
}
