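# MF_WPOAnim_StaticMesh_MainTime is a VFX material function that provides a looping timer for world position offset animation on static meshes.
# It divides Time by the LoopDuration input and outputs the fractional part, giving a repeating 0 to 1 sawtooth.
# LoopDuration defaults to its preview value, and lower values give faster, shorter loops.
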MaterialFunction /Game/Environments/Hera/FX/Materials/MaterialFunctions/MF_WPOAnim_StaticMesh_MainTime
graph {
 properties {
  description="Looping WPO timer, useful for looping movement on static meshes "
  userExposedCaption=VFX WPO Loop Time
  libraryCategoriesText [VFX]
 }
 nodes [
  n1 = Time()
  n2 = FunctionInput() {
   bUsePreviewValueAsDefault=true
   description="Set LoopDuration to lower value=faster speed and shorter loop time"
   inputName=LoopDuration
   inputType=FunctionInput_Scalar
   previewValue=<3 0 0 1>
  }
  FunctionOutput(None=Frac(None=Divide(A=@n1, B=@n2)))
 ]
}
