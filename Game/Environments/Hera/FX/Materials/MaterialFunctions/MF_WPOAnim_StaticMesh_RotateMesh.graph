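# Material function for VFX that spins a whole static mesh continuously through World Position Offset.
# RotateAboutAxis turns world position around the actor position, using a RotationAxis vector parameter (default Z) transformed from local space.
# The angle is time multiplied by a speed interpolated between RotationSpeed and RotationSpeed_Max.
# Vertex color red, remapped between the LerpMin and LerpMax inputs, drives that interpolation to vary speed.
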MaterialFunction /Game/Environments/Hera/FX/Materials/MaterialFunctions/MF_WPOAnim_StaticMesh_RotateMesh
graph {
 properties {
  description="Looping WPO timer, useful for looping movement on static meshes "
  userExposedCaption=VFX WPO Rotate mesh
  libraryCategoriesText [VFX]
 }
 nodes [
  n1 = VectorParameter() {parameterName=RotationAxis, defaultValue=<rgba 0 0 1 0>}
  n2 = Time()
  n3 = FunctionInput() {
   bUsePreviewValueAsDefault=true
   description=Rotates the whole mesh
   inputName=RotationSpeed
   inputType=FunctionInput_Scalar
   previewValue=<0.05 0 0 1>
  }
  n4 = FunctionInput() {
   bUsePreviewValueAsDefault=true
   description="Rotates the whole mesh, max value"
   inputName=RotationSpeed_Max
   inputType=FunctionInput_Scalar
   sortPriority=1
   previewValue=<0.1 0 0 1>
  }
  n5 = LinearInterpolate(A=@n3, B=@n4, Alpha=NamedRerouteUsage())
  FunctionOutput(None=RotateAboutAxis(
    NormalizedRotationAxis=Transform(None=@n1.RGB) {transformSourceType=TRANSFORMSOURCE_Local}
    RotationAngle=Multiply(A=@n2, B=@n5)
    PivotPoint=ActorPositionWS()
    Position=WorldPosition().XYZ
   ))
  n6 = FunctionInput() {
   bUsePreviewValueAsDefault=true
   description="Min value to set vertex color to, controls all other random values"
   inputName=VertexColor_LerpMin
   inputType=FunctionInput_Scalar
   sortPriority=2
   previewValue=<0.1 0 0 0>
  }
  n7 = FunctionInput() {
   bUsePreviewValueAsDefault=true
   description="Max value to set vertex color to, controls all other random values"
   inputName=VertexColor_LerpMax
   inputType=FunctionInput_Scalar
   sortPriority=3
   previewValue=<1 0 0 0>
  }
  NamedRerouteDeclaration(None=LinearInterpolate(A=@n6, B=@n7, Alpha=VertexColor().R)) {name=ClampVertexColor, nodeColor=<rgba 0 0.1797 1 1>}
 ]
}
