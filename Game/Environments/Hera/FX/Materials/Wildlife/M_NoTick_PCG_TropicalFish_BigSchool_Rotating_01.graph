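# Masked lit material for a rotating school of tropical fish on Niagara mesh particles, with a school texture scaled by an Emissive parameter.
# Shared material functions build looping-time, curl-noise and mesh-rotation world position offsets as named reroutes.
# A panned gradient texture adds an XYZ wave, with vertex color blue varying its UV offset and speed.
# The transformed wave is summed with two reroute inputs, then transformed again.
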Material /Game/Environments/Hera/FX/Materials/Wildlife/M_NoTick_PCG_TropicalFish_BigSchool_Rotating_01
graph {
 properties {
  bAutomaticallySetUsageInEditor=false
  bUsedWithNiagaraMeshParticles=true
  blendMode=BLEND_Masked
  previewMesh=/Game/Environments/Hera/FX/Meshes/SM_NoTick_PCG_FishSpiraling_01
  usedShadingModels=Default Lit
 }
 nodes [
  MaterialFunctionCall!(Sine Phase=Time(), Value 1, Value 2) {materialFunction=/Engine/Functions/Engine_MaterialFunctions02/Utility/Sine_Remapped}
  Multiply(
   A=TextureSampleParameter2D() {parameterName=Texture, texture=/Game/Environments/Hera/FX/Textures/T_Fish_PCG_TropicalFish_School_01_D}.RGB
   B=ScalarParameter() {parameterName=Emissive}
  )
  n1 = Time()
  n2 = NamedRerouteUsage()
  NamedRerouteDeclaration(None=MaterialFunctionCall!(
    InputLoopingTimeHere=Divide(A=@n1, B=@n2)
    VertexColor_LerpMax=ScalarParameter() {defaultValue=1, group=Curl, parameterName="VertexColor_LerpMax (S)"}
    VertexColor_LerpMin=ScalarParameter() {group=Curl, parameterName="VertexColor_LerpMin (S)"}
    CurlSpeed=ScalarParameter() {defaultValue=100, group=Curl, parameterName="CurlSpeed (S)"}
    Curl_Intensity=ScalarParameter() {defaultValue=250, group=Curl, parameterName="Curl_Intensity (S)"}
    Curl_Intensity_High=ScalarParameter() {defaultValue=100, group=Curl, parameterName="Curl_Intensity_High (S)"}
    CurlTiling=ScalarParameter() {defaultValue=2500, group=Curl, parameterName="CurlTiling (S)"}
    CurlTiling_High=ScalarParameter() {defaultValue=5000, group=Curl, parameterName="CurlTiling_High (S)"}
   ) {materialFunction=/Game/Environments/Hera/FX/Materials/MaterialFunctions/MF_WPOAnim_StaticMesh_Curl}.WPO Curl) {name=WPO_CURL, nodeColor=<rgba 0.3594 0 1 1>}
  n3 = ScalarParameter() {
   defaultValue=3
   desc="Looping WPO timer, useful for looping movement on static meshes "
   group=_MainLoopDuration
   parameterName="LoopDuration (S)"
   sortPriority=0
  }
  NamedRerouteDeclaration(None=@n3) {name=WPO_Loop_Time_ForCurl, nodeColor=<rgba 0 1 0.0805 1>}
  NamedRerouteDeclaration(None=MaterialFunctionCall!(LoopDuration=@n3) {materialFunction=/Game/Environments/Hera/FX/Materials/MaterialFunctions/MF_WPOAnim_StaticMesh_MainTime}.Result) {name=WPO_Loop_Time, nodeColor=<rgba 0 1 0.0805 1>}
  NamedRerouteDeclaration(None=MaterialFunctionCall!(
    RotationSpeed=ScalarParameter() {defaultValue=-0.05, parameterName="RotationSpeed (S)"}
    RotationSpeed_Max=ScalarParameter() {defaultValue=-0.01, parameterName="RotationSpeed_Max (S)"}
    VertexColor_LerpMin=ScalarParameter() {parameterName="VertexColor_LerpMin (S)_1"}
    VertexColor_LerpMax=ScalarParameter() {defaultValue=1, parameterName="VertexColor_LerpMax (S)_1"}
   ) {materialFunction=/Game/Environments/Hera/FX/Materials/MaterialFunctions/MF_WPOAnim_StaticMesh_RotateMesh}.Result) {name=WPO_RotateMesh, nodeColor=<rgba 0 0.8828 1 1>}
  SmoothStep()
  SphereMask()
  n4 = NamedRerouteUsage()
  n5 = NamedRerouteUsage()
  n6 = VertexColor()
  n7 = Multiply(
   A=Add(A=@n6.B, B=TextureCoordinate())
   B=ComponentMask(None=VectorParameter() {group=WPO, parameterName=TextureWPO_UVtiling, defaultValue=<rgba 1 1 0 0>}.RGB) {g=true, r=true}
  )
  n8 = ScalarParameter() {defaultValue=1, parameterName=TextureWPO_UVTiling_Mult}
  n9 = VectorParameter() {group=WPO, parameterName=TextureWPO_Speed, defaultValue=<rgba 1 0 0 0>}
  n10 = AppendVector(
   A=Add(A=LinearInterpolate(Alpha=@n6.B) {constA=0.15, constB=0.35}, B=ComponentMask(None=@n9.RGB) {r=true})
   B=ComponentMask(None=@n9.RGB) {g=true}
  )
  n11 = ScalarParameter() {defaultValue=1, parameterName=TextureWPO_Speed_Mult}
  n12 = TextureSample(UVs=Panner(Coordinate=Multiply(A=@n7, B=@n8), Speed=Multiply(A=@n10, B=@n11)) {speedX=0.15}) {texture=/Game/Global/Textures/Default/Masks_and_Gradients/WeaponGradients/T_Gradient_DesertSky_Color}
  n13 = Multiply(A=@n12.R, B=ScalarParameter() {defaultValue=1, group=WPO, parameterName=WPO_HeightX}) {constB=0.5}
  n14 = Multiply(A=@n12.R, B=ScalarParameter() {defaultValue=1, group=WPO, parameterName=WPO_HeightY}) {constB=0.5}
  n15 = ScalarParameter() {defaultValue=1, group=WPO, parameterName=WPO_HeightZ}
  n16 = Multiply(
   A=AppendVector(A=AppendVector(A=@n13, B=@n14), B=Multiply(A=@n12.R, B=@n15) {constB=0.6})
   B=ScalarParameter() {defaultValue=1, group=WPO, parameterName=WPO_Wave_Intenstiy}
  ) {constB=80}
  Transform(None=Add(A=Add(A=@n4, B=@n5), B=Transform(None=@n16) {transformSourceType=TRANSFORMSOURCE_Local})) {transformSourceType=TRANSFORMSOURCE_World}
 ]
}
