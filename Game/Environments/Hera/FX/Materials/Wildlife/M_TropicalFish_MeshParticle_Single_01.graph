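# Masked, lit material for tropical fish Niagara mesh particles.
# Particle color tints the hue-shifted fish texture, while dynamic parameters drive hue, emissive, green-channel night glow and Bayer-dithered alpha.
# World position offset combines looping curl motion from a shared function with a wave from a panning gradient texture.
# Blue vertex color varies the wave's UVs and speed, and a second dynamic parameter adjusts its speed and intensity.
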Material /Game/Environments/Hera/FX/Materials/Wildlife/M_TropicalFish_MeshParticle_Single_01
graph {
 properties {
  bAutomaticallySetUsageInEditor=false
  bUsedWithNiagaraMeshParticles=true
  blendMode=BLEND_Masked
  previewMesh=/Game/Environments/Hera/FX/Meshes/SM_Fish_Tropical_01
  usedShadingModels=Default Lit
 }
 nodes [
  n1 = TextureSampleParameter2D() {parameterName=Texture, texture=/Game/Environments/Hera/FX/Textures/T_Fish_PCG_TropicalFish_School_01_D}
  n2 = DynamicParameter() {paramNames [Emissive,NightGlow,Hue,Alpha]}
  Multiply(A=Multiply(A=@n1.RGB, B=@n2.Emissive), B=LinearInterpolate(B=Multiply(A=@n1.G, B=@n2.NightGlow), Alpha=@n2.NightGlow) {constA=1})
  n3 = ParticleColor()
  Multiply(
   A=@n3.A
   B=MaterialFunctionCall!(In=@n2.Alpha, Iterations, Scale) {materialFunction=/Game/Packages/Fortress_SharedMaterials/Base_Material_Functions/MF_DitherBayerNoTex}.Result
  )
  Multiply()
  Multiply(
   A=MaterialFunctionCall(Hue Shift Percentage=@n2.Hue, Texture=@n1.RGB) {materialFunction=&HueShift}.Result
   B=@n3.RGB
  )
  n4 = Time()
  n5 = NamedRerouteUsage()
  NamedRerouteDeclaration(None=MaterialFunctionCall!(
    InputLoopingTimeHere=Divide(A=@n4, B=@n5)
    VertexColor_LerpMax=ScalarParameter() {defaultValue=1, group=Curl, parameterName="VertexColor_LerpMax (S)"}
    VertexColor_LerpMin=ScalarParameter() {group=Curl, parameterName="VertexColor_LerpMin (S)"}
    CurlSpeed=ScalarParameter() {defaultValue=100, group=Curl, parameterName="CurlSpeed (S)"}
    Curl_Intensity=ScalarParameter() {defaultValue=250, group=Curl, parameterName="Curl_Intensity (S)"}
    Curl_Intensity_High=ScalarParameter() {defaultValue=100, group=Curl, parameterName="Curl_Intensity_High (S)"}
    CurlTiling=ScalarParameter() {defaultValue=2500, group=Curl, parameterName="CurlTiling (S)"}
    CurlTiling_High=ScalarParameter() {defaultValue=5000, group=Curl, parameterName="CurlTiling_High (S)"}
   ) {materialFunction=/Game/Environments/Hera/FX/Materials/MaterialFunctions/MF_WPOAnim_StaticMesh_Curl}.WPO Curl) {name=WPO_CURL, nodeColor=<rgba 0.3594 0 1 1>}
  n6 = ScalarParameter() {
   defaultValue=3
   desc="Looping WPO timer, useful for looping movement on static meshes "
   group=_MainLoopDuration
   parameterName="LoopDuration (S)"
   sortPriority=0
  }
  NamedRerouteDeclaration(None=@n6) {name=WPO_Loop_Time_ForCurl, nodeColor=<rgba 0 1 0.0805 1>}
  NamedRerouteDeclaration(None=MaterialFunctionCall!(LoopDuration=@n6) {materialFunction=/Game/Environments/Hera/FX/Materials/MaterialFunctions/MF_WPOAnim_StaticMesh_MainTime}.Result) {name=WPO_Loop_Time, nodeColor=<rgba 0 1 0.0805 1>}
  RgbToHsv!(None=@n1.RGB)
  n7 = VertexColor()
  n8 = Multiply(
   A=Add(A=@n7.B, B=TextureCoordinate())
   B=ComponentMask(None=VectorParameter() {group=WPO, parameterName=TextureWPO_UVtiling, defaultValue=<rgba 1 1 0 0>}.RGB) {g=true, r=true}
  )
  n9 = ScalarParameter() {defaultValue=1, parameterName=TextureWPO_UVTiling_Mult}
  n10 = VectorParameter() {group=WPO, parameterName=TextureWPO_Speed, defaultValue=<rgba 1 0 0 0>}
  n11 = AppendVector(
   A=Add(A=LinearInterpolate(Alpha=@n7.B) {constA=0.15, constB=0.35}, B=ComponentMask(None=@n10.RGB) {r=true})
   B=ComponentMask(None=@n10.RGB) {g=true}
  )
  n12 = Add(
   A=DynamicParameter() {parameterIndex=1, paramNames [WPO_Speed, WPO_Intensity, Param3, Param4]}.WPO_Speed
   B=ScalarParameter() {defaultValue=1, parameterName=TextureWPO_Speed_Mult}
  )
  n13 = TextureSample(UVs=Panner(Coordinate=Multiply(A=@n8, B=@n9), Speed=Multiply(A=@n11, B=@n12)) {speedX=0.15}) {texture=/Game/Global/Textures/Default/Masks_and_Gradients/WeaponGradients/T_Gradient_DesertSky_Color}
  n14 = Multiply(A=@n13.R, B=ScalarParameter() {defaultValue=1, group=WPO, parameterName=WPO_HeightX}) {constB=0.5}
  n15 = Multiply(A=@n13.R, B=ScalarParameter() {defaultValue=1, group=WPO, parameterName=WPO_HeightY}) {constB=0.5}
  n16 = ScalarParameter() {defaultValue=1, group=WPO, parameterName=WPO_HeightZ}
  n17 = Multiply(
   A=AppendVector(A=AppendVector(A=@n14, B=@n15), B=Multiply(A=@n13.R, B=@n16) {constB=0.6})
   B=Multiply(
    A=ScalarParameter() {defaultValue=1, group=WPO, parameterName=WPO_Wave_Intenstiy}
    B=DynamicParameter() {parameterIndex=1, paramNames [WPO_Speed, WPO_Intensity, Param3, Param4]}.WPO_Intensity
   )
  ) {constB=80}
  Transform(None=Add(A=NamedRerouteUsage(), B=Transform(None=@n17) {transformSourceType=TRANSFORMSOURCE_Local})) {transformSourceType=TRANSFORMSOURCE_World}
 ]
}
